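# Masked unlit particle material for Niagara sprites, ribbons and meshes, in the VehicleCosmetics Mutable FX set.
# A main texture parameter, defaulting to an impact gradient mask, passes through adjustable desaturation and contrast, then is tinted by particle color.
# Particle alpha multiplies either the texture's alpha channel or the desaturated color, chosen by a static switch.
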Material /VehicleCosmetics/Mutable/FX_Materials/SimpleParticle_MAT
graph {
 properties {
  bAutomaticallySetUsageInEditor=false
  bUsedWithNiagaraMeshParticles=true
  bUsedWithNiagaraRibbons=true
  bUsedWithNiagaraSprites=true
  blendMode=BLEND_Masked
  shadingModel=MSM_Unlit
  usedShadingModels=Unlit
 }
 nodes [
  n1 = ParticleColor()
  n2 = TextureSampleParameter2D() {group=Texture, parameterName=MainTex, texture=/DealershipSharedLibrary/FX/Textures/Gradient_Masks/T_VCVFX_ImpactFilled_01_A}
  n3 = Desaturation(None=@n2.RGB, Fraction=ScalarParameter() {parameterName=Desaturation})
  Multiply(A=@n1.RGB, B=Power(Base=@n3, Exp=ScalarParameter() {defaultValue=1, parameterName=Contrast}))
  Multiply(A=@n1.A, B=StaticSwitchParameter(True=@n2.A, False=@n3) {defaultValue=true, group=Alpha, parameterName=UseAlphaChannel?})
 ]
}
